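; Player controller component that handles camera and streaming for the ToyFox pilot cinematic.
; Replicated control rotation removes the pilot teleport camera modifier, ignores look input and sets the rotation the camera has when the sequence ends.
; Replicated stream location overrides the streaming source, awaiting streaming without loading screen or fade.
; These handlers run only for cosmetic execution; EndPlay restores look input and streaming, and DestroyComponent acts only on authority.
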(event Custom|OnRep_ControlRotation
  ;[0.02259 0.15 0.1305 0.5] OnRep_ControlRotation: Control rotation to have the camera at the end of the camera sequence has been received.
  (Networking|CanExecuteCosmeticEvents
    (:True
      (if (not (Components|IsComponentBeingDestroyed))
        (Fortnite|FortShippingLog self "OnRep_ControlRotation: invoked.")
        (bind _asfort_player_controller (Utilities|Casting|CastToFortPlayerController (Components|GetOwner)))
        (Utilities|FlowControl|Sequence
          (:then_0
            (bind _playercameramanager (Variables|/Script/Engine.PlayerController|GetPlayerCameraManager _asfort_player_controller)
              (:then
                (bind _returnvalue (Game|Player|FindCameraModifierbyClass _playercameramanager "/ToyFoxGameplay/Gameplay/PilotTeleport/ToyFoxPilotTeleportCameraModifier.ToyFoxPilotTeleportCameraModifier_C"))
                (Utilities|IsValid _returnvalue
                  (:Is%20Valid
                    (Game|Player|RemoveCameraModifier _playercameramanager _returnvalue))))
              (:else)))
          (:then_1
            (Input|SetIgnoreLookInput _asfort_player_controller true)
            (Fortnite|KeepCGPathOptimizations _asfort_player_controller 0 1.0) ; Check code logic for reasoning.
            (Pawn|SetControlRotation _asfort_player_controller (Variables|Transient|GetControlRotation))))))))

(event Custom|OnRep_StreamLocationOverride
  ;[0.1145 0.01308 0.15 0.5] OnRep_StreamLocationOverride: Stream around the input location while the cinematic is playing.
  (Networking|CanExecuteCosmeticEvents
    (:True
      (if (not (Components|IsComponentBeingDestroyed))
        (Fortnite|FortShippingLog self "OnRep_StreamLocationOverride: invoked.")
        (bind _asfort_player_controller_athena (Utilities|Casting|CastToFortPlayerControllerAthena (Components|GetOwner)))
        (Streaming|SetStreamingSourceOverride _asfort_player_controller_athena (Utilities|Struct|MakeStreamingSourceLocationRotationOverride 0 (Variables|Transient|GetStreamLocationOverride)) (struct :bAwaitStreaming true :bShowLoadingScreen false :bFadeCamera false))))))

(event Custom|DestroyComponent
  ;[0.15 0.04143 0 0.5] DestroyComponent: Call the K2_DestroyComponent so it runs successfully.
  (Networking|SwitchHasAuthority
    (:Authority
      (Components|DestroyComponent))))

(event EventEndPlay
  (params (EndPlayReason :type EEndPlayReason%20Enum :default "Destroyed"))
  ;[0.15 0.08039 0 0.5] EndPlay: Clean up.
  (Utilities|FlowControl|Sequence
    (:then_0
      (Networking|CanExecuteCosmeticEvents
        (:True
          (Fortnite|FortShippingLog self "EndPlay: invoked.")
          (bind _asfort_player_controller_athena (Utilities|Casting|CastToFortPlayerControllerAthena (Components|GetOwner)))
          (Input|SetIgnoreLookInput _asfort_player_controller_athena false)
          (Streaming|ResetStreamingSourceOverride _asfort_player_controller_athena (struct :bAwaitStreaming true :bShowLoadingScreen false :bFadeCamera false)))))
    (:then_1
      (|CallParentFunction EndPlayReason))))
